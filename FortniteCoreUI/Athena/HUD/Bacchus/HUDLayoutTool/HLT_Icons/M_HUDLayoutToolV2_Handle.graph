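# Translucent UI material for a handle in the HUD Layout Tool V2, drawing a tinted outline with an optional icon.
# A center-scaled UV mask splits the background into a border and a middle with adjustable opacity, tinted by Color.
# A switchable, separately scaled icon texture blends over it, confined to the middle by that mask.
# Final opacity takes the lighter of background and icon alpha.
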Material /FortniteCoreUI/Athena/HUD/Bacchus/HUDLayoutTool/HLT_Icons/M_HUDLayoutToolV2_Handle
graph {
 properties {
  bUsedWithStaticMesh=false
  blendMode=BLEND_Translucent
  materialDomain=MD_UI
  pixelDepthOffsetMode=PDOM_Legacy
  usedShadingModels=Default Lit
 }
 nodes [
  n1 = MaterialFunctionCall(UVs=TextureCoordinate(), Texture Scale=ScalarParameter() {defaultValue=0.6, group=Background, parameterName=OutlineScale, sliderMax=1}) {materialFunction=&ScaleUVsByCenter}
  n2 = MaterialFunctionCall!(
   Blend=Multiply(A=@n1.0-1 mask, B=ScalarParameter() {defaultValue=1, group=Background, parameterName=Middle Opacity, sliderMax=1})
   Base=OneMinus(None=@n1.0-1 mask)
  ) {materialFunction=/Engine/Functions/Engine_MaterialFunctions03/Blends/Blend_Lighten}
  n3 = VectorParameter() {parameterName=Color, defaultValue=<rgba 0 0.117 1 1>}
  n4 = TextureSampleParameter2D(UVs=MaterialFunctionCall(
    UVs=TextureCoordinate()
    Texture Scale=ScalarParameter() {defaultValue=0.55, group=Icon, parameterName=Icon Scale, sliderMax=10, sortPriority=4}
   ) {materialFunction=&ScaleUVsByCenter}.UVs) {group=Icon, parameterName=Icon, sortPriority=2, texture=&/T-HLT-Handle}
  n5 = Constant()
  n6 = ScalarParameter() {defaultValue=0.14, group=Icon, parameterName=IconOpacity, sliderMax=1, sortPriority=3}
  n7 = StaticSwitchParameter(True=Multiply(A=@n4.A, B=Multiply(A=@n1.0-1 mask, B=@n6)), False=@n5) {defaultValue=true, group=Icon, parameterName=Icon, sortPriority=1}
  LinearInterpolate(
   A=Multiply(A=@n2.Result, B=@n3.RGB)
   B=StaticSwitchParameter(True=@n4.RGB, False=@n5) {defaultValue=true, group=Icon, parameterName=Icon, sortPriority=1}
   Alpha=@n7
  )
  MaterialFunctionCall(Blend=@n7, Base=Saturate(None=Multiply(A=@n3.A, B=@n2.Result))) {materialFunction=&Blend_Lighten}
 ]
}
